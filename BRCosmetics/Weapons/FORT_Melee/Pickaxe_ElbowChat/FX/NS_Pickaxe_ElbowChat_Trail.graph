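# Niagara swing trail for the ElbowChat pickaxe, built from three short-lived ribbon emitters.
# Shared FN anim-trail modules spawn and update each ribbon from bones or sockets on the system's skeletal mesh.
# Screen-facing and side-vector ribbons form the near-black body, while a thinner, shorter-lived detail ribbon fades from red, offset further by camera offset.
# Curves shape color, alpha and width, and random dynamic material parameters vary the side and detail ribbons.
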NiagaraSystem /BRCosmetics/Weapons/FORT_Melee/Pickaxe_ElbowChat/FX/NS_Pickaxe_ElbowChat_Trail
dependencies {
 usedDataInterfaces [&NiagaraDataInterfaceColorCurve, &NiagaraDataInterfaceCurve, &NiagaraDataInterfaceSkeletalMesh]
 usedDynamicInputs [&Frac_Float, &Multiply_Float]
 usedModules [
  &self:SystemSpawnScript.SystemScriptSource.SystemScriptGraph.NiagaraNodeAssignment_2.SetVariables_#1
  &self:SystemSpawnScript.SystemScriptSource.SystemScriptGraph.NiagaraNodeAssignment_1.SetVariables_#2
  &self:SystemSpawnScript.SystemScriptSource.SystemScriptGraph.NiagaraNodeAssignment_0.SetVariables_#3
  &SystemState
 ]
}
emitters {
 Ribbon_Facing_Screen {
  emitterUpdateScript [
   &EmitterState(Life Cycle Mode=::NewEnumerator1, Loop Behavior=::NewEnumerator1, Loop Duration=0.2, UseLoopDelay=false)
   &FN_SpawnAnimTrail()
  ]
  particleSpawnScript [
   &FN_InitializeAnimTrailBoneIndecies(First Bone or Socket Index=1, Second Bone or Socket Index=0)
   &FN_InitializeAnimTrail(AnimTrailWidth=0.7, Color=<rgba 1 1 1 0.5>, Lifetime=0.134, SkeletalMesh=$System.Skeletal Mesh)
   &CameraOffset(Camera Offset Amount=10)
  ]
  particleUpdateScript [
   &FN_UpdateAnimTrail(SkeletalMesh=$System.Skeletal Mesh)
   &ScaleColor(
    Linear Color Curve=DataInterface() {propertyValues {
      AlphaCurve {keys [<at 0.2401 4 RCIM_Cubic>,<at 0.9527 0 RCIM_Cubic>]}
      BlueCurve {keys [{arriveTangent=0.0093, interpMode=RCIM_Cubic, time=0.0745, value=0.0076}, <at 0.3201 0>]}
      GreenCurve {keys [{arriveTangent=-0.0863, interpMode=RCIM_Cubic, time=0.0745, value=0.0076}, <at 0.3201 0>]}
      RedCurve {keys [{arriveTangent=-1.136, interpMode=RCIM_Cubic, time=0.0745, value=0.026}, <at 0.3201 0>]}
     }}
    Scale Mode=::NewEnumerator2
   )
   &ScaleRibbonWidth(Ribbon Width Scale=&FloatFromCurve(
     FloatCurve=DataInterface() {propertyValues {Curve {keys [
         <at 0 1.1686 RCIM_Cubic>
         {interpMode=RCIM_Cubic, tangentMode=RCTM_User, time=0.2, value=1.12}
         {interpMode=RCIM_Cubic, tangentMode=RCTM_User, time=1, value=1.25}
        ]}}}
     Scale Curve=1.22
    ))
  ]
  rendererClasses [&NiagaraRibbonRendererProperties]
  renderers [{rendererClass=&NiagaraRibbonRendererProperties}]
 }
 Ribbon_Facing_Screen_Detail_Red {
  emitterUpdateScript [
   &EmitterState(Life Cycle Mode=::NewEnumerator1, Loop Behavior=::NewEnumerator1, Loop Duration=0.2, UseLoopDelay=false)
   &FN_SpawnAnimTrail()
  ]
  particleSpawnScript [
   &FN_InitializeAnimTrailBoneIndecies()
   &FN_InitializeAnimTrail(AnimTrailWidth=0.33, Color=<rgba 1 1 1 0.5>, Lifetime=0.08, SkeletalMesh=$System.Skeletal Mesh)
   &CameraOffset(Camera Offset Amount=15)
   &DynamicMaterialParameters(
    Index 0 Param 1=$System.RandomFloatA
    Index 0 Param 2=$System.RandomFloatB
    Module.Param0WriteEnabled=true
    Module.Param1WriteEnabled=true
    Module.Param2WriteEnabled=false
    Module.Param3WriteEnabled=false
   )
  ]
  particleUpdateScript [
   &FN_UpdateAnimTrail(SkeletalMesh=$System.Skeletal Mesh)
   &ScaleColor(
    Linear Color Curve=DataInterface() {propertyValues {
      AlphaCurve {keys [
        {arriveTangent=-3.2415, interpMode=RCIM_Cubic, leaveTangent=-3.2415, tangentMode=RCTM_User, time=0.0037, value=1.1}
        {arriveTangentWeight=0.326, interpMode=RCIM_Cubic, tangentMode=RCTM_User, time=0.831, value=0}
       ]}
      BlueCurve {keys [<at 0 0 RCIM_Cubic>, <at 0.0173 0>, {interpMode=RCIM_Cubic, leaveTangent=0.0191, time=0.2168, value=0}, <at 0.4102 0.006>]}
      GreenCurve {keys [
        {arriveTangent=-2.6767, interpMode=RCIM_Cubic, time=0, value=0.008}
        <at 0.0173 0>
        {interpMode=RCIM_Cubic, leaveTangent=0.0176, time=0.2168, value=0.0208}
        <at 0.4102 0.0055>
       ]}
      RedCurve {keys [
        {arriveTangent=-6.9405, interpMode=RCIM_Cubic, time=0, value=1}
        <at 0.0173 1>
        {interpMode=RCIM_Cubic, leaveTangent=-3.1351, time=0.2168, value=0.25}
        <at 0.4102 0.0156>
       ]}
     }}
    Scale Mode=::NewEnumerator2
   )
   &ScaleRibbonWidth(Ribbon Width Scale=&FloatFromCurve(
     FloatCurve=DataInterface() {propertyValues {Curve {keys [{interpMode=RCIM_Cubic, tangentMode=RCTM_User, time=0, value=0.5}, {interpMode=RCIM_Cubic, leaveTangent=1.3038, time=1, value=1}]}}}
     Scale Curve=1.2
    ))
  ]
  rendererClasses [&NiagaraRibbonRendererProperties]
  renderers [{rendererClass=&NiagaraRibbonRendererProperties}]
 }
 Ribbon_SideVector {
  emitterUpdateScript [
   &EmitterState(Life Cycle Mode=::NewEnumerator1, Loop Behavior=::NewEnumerator1, Loop Duration=0.2, UseLoopDelay=false)
   &FN_SpawnAnimTrail()
  ]
  particleSpawnScript [
   &FN_InitializeAnimTrailBoneIndecies(First Bone or Socket Index=1, Second Bone or Socket Index=0)
   &FN_InitializeAnimTrail(AnimTrailWidth=1.25, Color=<rgba 1 1 1 0.5>, Lifetime=0.134, SkeletalMesh=$System.Skeletal Mesh)
   &CameraOffset(Camera Offset Amount=10)
  ]
  particleUpdateScript [
   &FN_UpdateAnimTrail(SkeletalMesh=$System.Skeletal Mesh)
   &ScaleColor(
    Linear Color Curve=DataInterface() {propertyValues {
      AlphaCurve {keys [<at 0 4 RCIM_Cubic>,<at 0.5676 1>,<at 0.9527 0 RCIM_Cubic>]}
      BlueCurve {keys [{arriveTangent=0.1336, interpMode=RCIM_Cubic, time=0.009, value=0.0211}, <at 0.0657 0.0081>]}
      GreenCurve {keys [{arriveTangent=-0.4141, interpMode=RCIM_Cubic, time=0.009, value=0.0211}, <at 0.0657 0.0081>]}
      RedCurve {keys [{arriveTangent=-16.267, interpMode=RCIM_Cubic, time=0.009, value=0.05}, <at 0.0657 0.01>]}
     }}
    Scale Mode=::NewEnumerator2
   )
   &ScaleRibbonWidth(Ribbon Width Scale=&FloatFromCurve(
     FloatCurve=DataInterface() {propertyValues {Curve {keys [
         <at 0 1.4899 RCIM_Cubic>
         {interpMode=RCIM_Cubic, tangentMode=RCTM_User, time=0.2, value=1.1}
         {interpMode=RCIM_Cubic, tangentMode=RCTM_User, time=1, value=1.5}
        ]}}}
     Scale Curve=1.07
    ))
   &DynamicMaterialParameters(
    Index 0 Param 1=&RandomRangeFloat(Minimum=-0.25, Maximum=0.25)
    Index 0 Param 2=1
    Module.Param0WriteEnabled=true
    Module.Param1WriteEnabled=true
    Module.Param2WriteEnabled=false
    Module.Param3WriteEnabled=false
   )
  ]
  rendererClasses [&NiagaraRibbonRendererProperties]
  renderers [{rendererClass=&NiagaraRibbonRendererProperties}]
 }
}
summary {
 systemName=NS_Pickaxe_ElbowChat_Trail
 emitters [
  {emitterName=Ribbon_Facing_Screen, rendererClasses [&NiagaraRibbonRendererProperties]}
  {emitterName=Ribbon_SideVector, rendererClasses [&NiagaraRibbonRendererProperties]}
  {emitterName=Ribbon_Facing_Screen_Detail_Red, rendererClasses [&NiagaraRibbonRendererProperties]}
 ]
 userVariables [User.RibbonIDOffset {type=NiagaraInt32}]
}
