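# Physics asset for the ElbowChat pickaxe mesh, a two-handed melee weapon cosmetic in BRCosmetics.
# It holds a single body on the Melee_TwoHanded_Handle bone, shaped as one rotated capsule offset from the bone origin.
# It defines no constraints, so the axe has one simple collision volume with no jointed parts.
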PhysicsAsset /BRCosmetics/Weapons/FORT_Melee/Pickaxe_ElbowChat/Meshes/ElbowChat_Axe_PhysicsAsset
bodies {Melee_TwoHanded_Handle [{shapeType=Capsule, length=73.6719, radius=14.2025, center=<-1.2935 -0.6461 34.275>, rotation=<rot 0.5207 -90 -1.2577>}]}
constraints []
